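# Lit material for the hamburger head barrier, textured with the Durrr sign diffuse, normal and mask maps.
# World-aligned cracks darken the diffuse, blended toward the clean diffuse as Code_DestructionAmount approaches its default of one.
# The mask's blue channel blends two roughness scalars, and the assembled material attributes feed MF_CreativeMoveTool.
# A Fortnite_HitEffects call takes damage parameters and diffuse, but its output feeds none of these nodes.
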Material /Game/Effects/Fort_Effects/Materials/Barrier/M_Barrier_Base_Head_Hamburger
graph {
 properties {bUseMaterialAttributes=true, pixelDepthOffsetMode=PDOM_Legacy, usedShadingModels=Default Lit}
 nodes [
  n1 = MaterialFunctionCall(
   TextureObject=TextureObject() {texture=/Game/Global/Textures/Default/Masks_and_Gradients/Cracks_02}
   TextureSize=Constant() {r=350}
   WorldPosition=WorldPosition() {worldPositionShaderOffset=WPT_ExcludeAllShaderOffsets}.XYZ
  ) {materialFunction=&WorldAlignedTexture}
  n2 = TextureSampleParameter2D() {parameterName=Diffuse, texture=/Game/Environments/Props/Signs/Textures/Sign_Durrr_D}
  n3 = Multiply(A=OneMinus(None=ComponentMask(None=@n1.XYZ Texture) {r=true}), B=@n2.RGB)
  n4 = ScalarParameter() {defaultValue=1, group=z_code, parameterName=Code_DestructionAmount}
  n5 = ScalarParameter() {defaultValue=0.1, parameterName=A}
  n6 = ScalarParameter() {defaultValue=1, parameterName=B}
  n7 = TextureSample() {samplerType=SAMPLERTYPE_Masks, texture=/Game/Environments/Props/Signs/Textures/Sign_Durrr_S}
  MaterialFunctionCall(MA In=SetMaterialAttributes!(
    MaterialAttributes
    Base Color=LinearInterpolate(A=@n3, B=@n2.RGB, Alpha=@n4)
    Roughness=LinearInterpolate(A=@n5, B=@n6, Alpha=@n7.B)
    Normal=TextureSampleParameter2D() {parameterName=Normal, samplerType=SAMPLERTYPE_Normal, texture=/Game/Environments/Props/Signs/Textures/Sign_Durrr_N}.RGB
   ) {attributeSetTypes [#1, #2, #3]}) {materialFunction=&MF_CreativeMoveTool}
  MaterialFunctionCall!(
   WSDeformationCenter=VectorParameter() {group=z_code, parameterName=WSDeformationCenter}.RGB
   Code_DestructionAmount=@n4
   Code_UseUvForDamage
   Code_UseHitEffects=StaticBoolParameter() {defaultValue=true, group=z_code, parameterName=Code_UseHitEffects}
   Code_Emissive
   WSDamageRadius=ScalarParameter() {defaultValue=1, group=z_code, parameterName=WSDamageRadius}
   CrackDamageShader
   Diffuse=@n2.RGB
   ScalarModDamageCracks
   ScalarOffsetforDamageCracks
   ExposeLerpFunctionForInvertingTheHitDistanceCalc
   InvertDamageFalloff
   CrackWorldPosition
   WorldPosition
   DistanceCullFadeAnimationOffset=ScalarParameter() {group=z_code, parameterName=DistanceCullFadeAnimationOffset}
  ) {materialFunction=/Game/Packages/jl_tests/Fortnite_HitEffects}
  ScalarParameter() {group=z_code, parameterName=Code_HitAnimation}
  VectorParameter() {group=z_code, parameterName=Code_DeformationVector, defaultValue=<rgba 0 0 1 0>}
 ]
}
